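# Physics asset for the left leg mesh of the male medium ChefTartar cosmetic body, set up for physics-driven shorts.
# Four tiny base shorts bodies attach to the thigh with free angular motion, each anchoring a dynamic shorts capsule with narrow swing and very tight twist limits.
# Larger thigh and calf-twist capsules give the leg its volume, with the calf also joined to the thigh by a free angular constraint.
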PhysicsAsset /BRCosmetics/Characters/Player/Male/Medium/Bodies/M_MED_ChefTartar/Meshes/M_MED_ChefTartar_Left_leg_Physics
bodies {
 base_shorts_l_01 [{shapeType=Capsule, length=0.505, radius=0.505, shapeName=base_shorts_l_01_capsule, rotation=<rot 0 -90 0>}]
 base_shorts_l_02 [{shapeType=Capsule, length=0.505, radius=0.505, shapeName=base_shorts_l_01_capsule, rotation=<rot 0 -90 0>}]
 base_shorts_l_03 [{shapeType=Capsule, length=0.505, radius=0.505, shapeName=base_shorts_l_01_capsule, rotation=<rot 0 -90 0>}]
 base_shorts_l_04 [{shapeType=Capsule, length=0.505, radius=0.505, shapeName=base_shorts_l_01_capsule, rotation=<rot 0 -90 0>}]
 calf_twist_02_l [{shapeType=Capsule, length=12, radius=4, shapeName=calf_twist_02_l_capsule, center=<-6 -13.3054 0>}]
 dyn_shorts_l_01 [{shapeType=Capsule, length=6, radius=1.5, shapeName=dyn_shorts_l_01_capsule, center=<5 0 0>, rotation=<rot 90 0 0>}]
 dyn_shorts_l_02 [{shapeType=Capsule, length=12, radius=1.5, shapeName=dyn_shorts_l_01_capsule, center=<7 0 0>, rotation=<rot 90 0 0>}]
 dyn_shorts_l_03 [{shapeType=Capsule, length=6, radius=1.5, shapeName=dyn_shorts_l_01_capsule, center=<5 0 0>, rotation=<rot 90 0 0>}]
 dyn_shorts_l_04 [{shapeType=Capsule, length=6, radius=1.5, shapeName=dyn_shorts_l_01_capsule, center=<5 0 0>, rotation=<rot 90 0 0>}]
 thigh_l [{shapeType=Capsule, length=20, radius=5, shapeName=thigh_r_capsule, center=<-17.1823 -0.3181 0.4591>, rotation=<rot 90 0 0>}]
}
constraints [
 {
  bone1Name=base_shorts_l_01
  bone2Name=thigh_l
  swing1LimitDegrees=45
  swing1Motion=Free
  swing2LimitDegrees=45
  swing2Motion=Free
  twistLimitDegrees=45
  twistMotion=Free
 }
 {
  bone1Name=base_shorts_l_02
  bone2Name=thigh_l
  swing1LimitDegrees=45
  swing1Motion=Free
  swing2LimitDegrees=45
  swing2Motion=Free
  twistLimitDegrees=45
  twistMotion=Free
 }
 {
  bone1Name=base_shorts_l_03
  bone2Name=thigh_l
  swing1LimitDegrees=45
  swing1Motion=Free
  swing2LimitDegrees=45
  swing2Motion=Free
  twistLimitDegrees=45
  twistMotion=Free
 }
 {
  bone1Name=base_shorts_l_04
  bone2Name=thigh_l
  swing1LimitDegrees=45
  swing1Motion=Free
  swing2LimitDegrees=45
  swing2Motion=Free
  twistLimitDegrees=45
  twistMotion=Free
 }
 {bone1Name=dyn_shorts_l_01, bone2Name=base_shorts_l_01, swing1LimitDegrees=20, swing2LimitDegrees=15, twistLimitDegrees=1}
 {bone1Name=dyn_shorts_l_02, bone2Name=base_shorts_l_02, swing1LimitDegrees=20, swing2LimitDegrees=20, twistLimitDegrees=1}
 {bone1Name=dyn_shorts_l_03, bone2Name=base_shorts_l_03, swing1LimitDegrees=20, swing2LimitDegrees=10, twistLimitDegrees=1}
 {bone1Name=dyn_shorts_l_04, bone2Name=base_shorts_l_04, swing1LimitDegrees=20, swing2LimitDegrees=7, twistLimitDegrees=1}
 {
  bone1Name=calf_twist_02_l
  bone2Name=thigh_l
  swing1LimitDegrees=45
  swing1Motion=Free
  swing2LimitDegrees=45
  swing2Motion=Free
  twistLimitDegrees=45
  twistMotion=Free
 }
]
