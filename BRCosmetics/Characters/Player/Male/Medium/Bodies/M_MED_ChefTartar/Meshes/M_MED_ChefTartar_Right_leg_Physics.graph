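# Physics asset for the right leg of the male medium ChefTartar body cosmetic, defining bodies for the jointed shorts.
# Four thin dynamic-shorts capsules each pair with a tiny base capsule through constraints with narrow swing and locked or near-zero twist.
# The base capsules and a calf-twist capsule connect to the thigh capsule through constraints with free swing and twist.
# Larger capsules give the thigh and calf their volume.
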PhysicsAsset /BRCosmetics/Characters/Player/Male/Medium/Bodies/M_MED_ChefTartar/Meshes/M_MED_ChefTartar_Right_leg_Physics
bodies {
 base_shorts_r_01 [{shapeType=Capsule, length=0.505, radius=0.505, shapeName=base_shorts_r_01_capsule, rotation=<rot 0 -90 180>}]
 base_shorts_r_02 [{shapeType=Capsule, length=0.505, radius=0.505, shapeName=base_shorts_r_01_capsule, rotation=<rot 0 -90 180>}]
 base_shorts_r_03 [{shapeType=Capsule, length=0.505, radius=0.505, shapeName=base_shorts_r_01_capsule, rotation=<rot 0 -90 180>}]
 base_shorts_r_04 [{shapeType=Capsule, length=0.505, radius=0.505, shapeName=base_shorts_r_01_capsule, rotation=<rot 0 -90 180>}]
 calf_twist_02_r [{shapeType=Capsule, length=12, radius=4, shapeName=calf_twist_02_r_capsule, center=<6 13.3054 0>, rotation=<rot 0 0 -180>}]
 dyn_shorts_r_01 [{shapeType=Capsule, length=6, radius=1.5, shapeName=dyn_shorts_r_01_capsule, center=<5 0 0>, rotation=<rot 90 180 0>}]
 dyn_shorts_r_02 [{shapeType=Capsule, length=12, radius=1.5, shapeName=dyn_shorts_r_01_capsule, center=<7 0 0>, rotation=<rot 90 180 0>}]
 dyn_shorts_r_03 [{shapeType=Capsule, length=6, radius=1.5, shapeName=dyn_shorts_r_01_capsule, center=<5 0 0>, rotation=<rot 90 180 0>}]
 dyn_shorts_r_04 [{shapeType=Capsule, length=6, radius=1.5, shapeName=dyn_shorts_r_01_capsule, center=<5 0 0>, rotation=<rot 90 180 0>}]
 thigh_r [{shapeType=Capsule, length=20, radius=5, shapeName=thigh_l_capsule, center=<17.1823 0.3181 -0.4591>, rotation=<rot 90 180 0>}]
}
constraints [
 {bone1Name=dyn_shorts_r_01, bone2Name=base_shorts_r_01, swing1LimitDegrees=15, swing2LimitDegrees=15, twistLimitDegrees=0, twistMotion=Locked}
 {bone1Name=dyn_shorts_r_02, bone2Name=base_shorts_r_02, swing1LimitDegrees=20, swing2LimitDegrees=20, twistLimitDegrees=1}
 {bone1Name=dyn_shorts_r_03, bone2Name=base_shorts_r_03, swing1LimitDegrees=20, swing2LimitDegrees=10, twistLimitDegrees=1}
 {bone1Name=dyn_shorts_r_04, bone2Name=base_shorts_r_04, swing1LimitDegrees=20, swing2LimitDegrees=7, twistLimitDegrees=1}
 {
  bone1Name=calf_twist_02_r
  bone2Name=thigh_r
  swing1LimitDegrees=45
  swing1Motion=Free
  swing2LimitDegrees=45
  swing2Motion=Free
  twistLimitDegrees=45
  twistMotion=Free
 }
 {
  bone1Name=base_shorts_r_04
  bone2Name=thigh_r
  swing1LimitDegrees=45
  swing1Motion=Free
  swing2LimitDegrees=45
  swing2Motion=Free
  twistLimitDegrees=45
  twistMotion=Free
 }
 {
  bone1Name=base_shorts_r_03
  bone2Name=thigh_r
  swing1LimitDegrees=45
  swing1Motion=Free
  swing2LimitDegrees=45
  swing2Motion=Free
  twistLimitDegrees=45
  twistMotion=Free
 }
 {
  bone1Name=base_shorts_r_02
  bone2Name=thigh_r
  swing1LimitDegrees=45
  swing1Motion=Free
  swing2LimitDegrees=45
  swing2Motion=Free
  twistLimitDegrees=45
  twistMotion=Free
 }
 {
  bone1Name=base_shorts_r_01
  bone2Name=thigh_r
  swing1LimitDegrees=45
  swing1Motion=Free
  swing2LimitDegrees=45
  swing2Motion=Free
  twistLimitDegrees=45
  twistMotion=Free
 }
]
