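# Post-copy cosmetic AnimBP for the ChefTartar character body, adding leg physics and shorts bone adjustments to the pose copied from the attached parent mesh.
# Both branches of a bool blend run Control Rig, then left and right leg RigidBody nodes: one in component space at partial alpha, the other in thigh-bone space.
# The thigh-space branch also stacks two passes of additive shorts bone offsets.
# Update computes Only_Crouching (stationary crouch) and a deprecated gravity override.
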AnimBlueprint /BRCosmetics/Characters/Player/Male/Medium/Bodies/M_MED_ChefTartar/Meshes/M_MED_ChefTartar_AnimBP
target_skeleton=&/M_MED_ChefTartar_Skeleton
graphs [
 AnimGraph {nodes [
   n1 = Control Rig(Source=Use cached pose 'Char_cache_Pose'()) {controlRigAssetReference {blueprintRigClass=&/M_MED_ChefTartar_CtrlRig.M_MED_ChefTartar_CtrlRig_C}}
   n2 = RigidBody!(ComponentPose=Local To Component(LocalPose=@n1), Alpha, OverrideWorldGravity:Vector="0,0,0", ExternalForce) -> (Pose) {
    bClampLinearTranslationLimitToRefPose=true
    bOverrideWorldGravity=true
    overridePhysicsAsset=&/M_MED_ChefTartar_Left_leg_Physics
    simulationSpace=BaseBoneSpace
    baseBoneRef {boneName=thigh_l}
    componentAppliedLinearAccClamp=<0 0 0>
    simSpaceSettings {
     maxAngularAcceleration=1
     maxAngularVelocity=1
     maxLinearAcceleration=7000
     maxLinearVelocity=5000
     worldAlpha=1
     externalLinearDragV=<0.2 0.2 0.2>
    }
   }
   n3 = RigidBody(ComponentPose=@n2, OverrideWorldGravity="0,0,0") {
    bClampLinearTranslationLimitToRefPose=true
    bOverrideWorldGravity=true
    overridePhysicsAsset=&/M_MED_ChefTartar_Right_leg_Physics
    simulationSpace=BaseBoneSpace
    baseBoneRef {boneName=thigh_r}
    componentAppliedLinearAccClamp=<0 0 0>
    simSpaceSettings {
     maxAngularAcceleration=1
     maxAngularVelocity=1
     maxLinearAcceleration=7000
     maxLinearVelocity=5000
     worldAlpha=1
     externalLinearDragV=<0.2 0.2 0.2>
    }
   }
   n4 = "Transform (Modify) Bone - Bone: base_shorts_l_04"!(ComponentPose=@n3, bAlphaBoolEnabled:bool=True, Translation="-10,0,0", Rotation="0,3,0", Scale="1,1,1") -> (Pose) {
    alphaInputType=Bool
    rotationMode=BMM_Additive
    rotationSpace=BCS_BoneSpace
    translationMode=BMM_Additive
    translationSpace=BCS_BoneSpace
    alphaBoolBlend {blendInTime=0.2, blendOutTime=0.3}
    boneToModify {boneName=base_shorts_l_04}
    rotation {yaw=3}
    translation {x=-10}
   }
   n5 = "Transform (Modify) Bone - Bone: base_shorts_l_03"!(ComponentPose=@n4, bAlphaBoolEnabled:bool=True, Translation="-8,0,0", Rotation="0,-3,0", Scale="1,1,1") -> (Pose) {
    alphaInputType=Bool
    rotationMode=BMM_Additive
    rotationSpace=BCS_BoneSpace
    translationMode=BMM_Additive
    translationSpace=BCS_BoneSpace
    alphaBoolBlend {blendInTime=0.2, blendOutTime=0.3}
    boneToModify {boneName=base_shorts_l_03}
    rotation {yaw=-3}
    translation {x=-8}
   }
   n6 = "Transform (Modify) Bone - Bone: base_shorts_l_02"!(ComponentPose=@n5, bAlphaBoolEnabled:bool=True, Translation="-20,0,0", Rotation="0,0,0", Scale="1,1,1") -> (Pose) {
    alphaInputType=Bool
    rotationMode=BMM_Additive
    rotationSpace=BCS_BoneSpace
    translationMode=BMM_Additive
    translationSpace=BCS_BoneSpace
    alphaBoolBlend {blendInTime=0.2, blendOutTime=0.3}
    boneToModify {boneName=base_shorts_l_02}
    translation {x=-20}
   }
   n7 = "Transform (Modify) Bone - Bone: base_shorts_r_04"!(ComponentPose=@n6, bAlphaBoolEnabled:bool=True, Translation="-6,0,0", Rotation="0,3,0", Scale="1,1,1") -> (Pose) {
    alphaInputType=Bool
    rotationMode=BMM_Additive
    rotationSpace=BCS_BoneSpace
    translationMode=BMM_Additive
    translationSpace=BCS_BoneSpace
    alphaBoolBlend {blendInTime=0.2, blendOutTime=0.3}
    boneToModify {boneName=base_shorts_r_04}
    rotation {yaw=3}
    translation {x=-6}
   }
   n8 = "Transform (Modify) Bone - Bone: base_shorts_r_03"!(ComponentPose=@n7, bAlphaBoolEnabled:bool=True, Translation="-20,0,0", Rotation="0,-3,0", Scale="1,1,1") -> (Pose) {
    alphaInputType=Bool
    rotationMode=BMM_Additive
    rotationSpace=BCS_BoneSpace
    translationMode=BMM_Additive
    translationSpace=BCS_BoneSpace
    alphaBoolBlend {blendInTime=0.2, blendOutTime=0.3}
    boneToModify {boneName=base_shorts_r_03}
    rotation {yaw=-3}
    translation {x=-20}
   }
   n9 = "Transform (Modify) Bone - Bone: base_shorts_r_02"!(ComponentPose=@n8, bAlphaBoolEnabled:bool=True, Translation="-15,0,0", Rotation="0,0,0", Scale="1,1,1") -> (Pose) {
    alphaInputType=Bool
    rotationMode=BMM_Additive
    rotationSpace=BCS_BoneSpace
    translationMode=BMM_Additive
    translationSpace=BCS_BoneSpace
    alphaBoolBlend {blendInTime=0.2, blendOutTime=0.3}
    boneToModify {boneName=base_shorts_r_02}
    translation {x=-15}
   }
   n10 = "Transform (Modify) Bone - Bone: base_shorts_l_04"(ComponentPose=@n9, bAlphaBoolEnabled=True, Translation="-10,0,0", Rotation="0,3,0", Scale="1,1,1") {
    alphaInputType=Bool
    rotationMode=BMM_Additive
    rotationSpace=BCS_BoneSpace
    translationMode=BMM_Additive
    translationSpace=BCS_BoneSpace
    alphaBoolBlend {blendInTime=0.2, blendOutTime=0.8}
    boneToModify {boneName=base_shorts_l_04}
    rotation {yaw=3}
    translation {x=-10}
   }
   n11 = "Transform (Modify) Bone - Bone: base_shorts_l_03"(ComponentPose=@n10, bAlphaBoolEnabled=True, Translation="-8,0,0", Rotation="0,-3,0", Scale="1,1,1") {
    alphaInputType=Bool
    rotationMode=BMM_Additive
    rotationSpace=BCS_BoneSpace
    translationMode=BMM_Additive
    translationSpace=BCS_BoneSpace
    alphaBoolBlend {blendInTime=0.2, blendOutTime=0.8}
    boneToModify {boneName=base_shorts_l_03}
    rotation {yaw=-3}
    translation {x=-8}
   }
   n12 = "Transform (Modify) Bone - Bone: base_shorts_l_02"(ComponentPose=@n11, bAlphaBoolEnabled=True, Translation="-20,0,0", Rotation="0,0,0", Scale="1,1,1") {
    alphaInputType=Bool
    rotationMode=BMM_Additive
    rotationSpace=BCS_BoneSpace
    translationMode=BMM_Additive
    translationSpace=BCS_BoneSpace
    alphaBoolBlend {blendInTime=0.2, blendOutTime=0.8}
    boneToModify {boneName=base_shorts_l_02}
    translation {x=-20}
   }
   n13 = "Transform (Modify) Bone - Bone: base_shorts_r_04"(ComponentPose=@n12, bAlphaBoolEnabled=True, Translation="-6,0,0", Rotation="0,3,0", Scale="1,1,1") {
    alphaInputType=Bool
    rotationMode=BMM_Additive
    rotationSpace=BCS_BoneSpace
    translationMode=BMM_Additive
    translationSpace=BCS_BoneSpace
    alphaBoolBlend {blendInTime=0.2, blendOutTime=0.8}
    boneToModify {boneName=base_shorts_r_04}
    rotation {yaw=3}
    translation {x=-6}
   }
   n14 = "Transform (Modify) Bone - Bone: base_shorts_r_03"(ComponentPose=@n13, bAlphaBoolEnabled=True, Translation="-20,0,0", Rotation="0,-3,0", Scale="1,1,1") {
    alphaInputType=Bool
    rotationMode=BMM_Additive
    rotationSpace=BCS_BoneSpace
    translationMode=BMM_Additive
    translationSpace=BCS_BoneSpace
    alphaBoolBlend {blendInTime=0.2, blendOutTime=0.8}
    boneToModify {boneName=base_shorts_r_03}
    rotation {yaw=-3}
    translation {x=-20}
   }
   n15 = "Transform (Modify) Bone - Bone: base_shorts_r_02"(ComponentPose=@n14, bAlphaBoolEnabled=True, Translation="-15,0,0", Rotation="0,0,0", Scale="1,1,1") {
    alphaInputType=Bool
    rotationMode=BMM_Additive
    rotationSpace=BCS_BoneSpace
    translationMode=BMM_Additive
    translationSpace=BCS_BoneSpace
    alphaBoolBlend {blendInTime=0.2, blendOutTime=0.8}
    boneToModify {boneName=base_shorts_r_02}
    translation {x=-15}
   }
   n16 = Control Rig(Source=Use cached pose 'Char_cache_Pose'()) {controlRigAssetReference {blueprintRigClass=&M_MED_ChefTartar_CtrlRig_C}}
   n17 = RigidBody(ComponentPose=Local To Component(LocalPose=@n16), Alpha=0.6, OverrideWorldGravity="0,0,0") {
    bClampLinearTranslationLimitToRefPose=true
    bOverrideWorldGravity=true
    overridePhysicsAsset=&M_MED_ChefTartar_Left_leg_Physics
    baseBoneRef {boneName=thigh_l}
    componentAppliedLinearAccClamp=<0 0 0>
    simSpaceSettings {
     maxAngularAcceleration=1
     maxAngularVelocity=1
     maxLinearAcceleration=7000
     maxLinearVelocity=5000
     worldAlpha=1
     externalLinearDragV=<0.2 0.2 0.2>
    }
   }
   n18 = RigidBody(ComponentPose=@n17, Alpha=0.6, OverrideWorldGravity="0,0,0") {
    bClampLinearTranslationLimitToRefPose=true
    bOverrideWorldGravity=true
    overridePhysicsAsset=&M_MED_ChefTartar_Right_leg_Physics
    baseBoneRef {boneName=thigh_r}
    componentAppliedLinearAccClamp=<0 0 0>
    simSpaceSettings {
     maxAngularAcceleration=1
     maxAngularVelocity=1
     maxLinearAcceleration=7000
     maxLinearVelocity=5000
     worldAlpha=1
     externalLinearDragV=<0.2 0.2 0.2>
    }
   }
   Output Pose(Result=Blend Poses by bool(BlendPose_0=Component To Local(ComponentPose=@n18), BlendPose_1=Component To Local(ComponentPose=@n15))) {layerGroup=None}
   Save cached pose 'Char_cache_Pose'(Pose=Copy Pose From Mesh() {bUseAttachedParent=true}) {cacheName=Char_cache_Pose}
  ]}
 EventGraph {nodes [
   Event BlueprintUpdateAnimation()
   > n1 = Gravity Override State Function DEPRECATED(CharacterPartAnimInstance=Self-Reference(), GameTimeInSeconds=GetGameTimeInSeconds(), GravityOverrideParams=Get Gravity Override Params())
   > Set Gravity Override(Gravity Override=@n1)
   n2 = Get bIsMoving2D()
   > Set Only_Crouching(Only_Crouching:bool=AND Boolean(A=NOT Boolean(A=@n2), B=Get bIsCrouching())) -> (Output_Get:bool)
  ]}
]
variables [
 Gravity Override {type=Vector}
 Gravity Override Params {type=GravityOverrideParamsStruct}
 Only_Crouching {type=boolean}
]
